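# Translucent, unlit, two-sided effect material for the Viper glider's FX mesh, usable on skeletal meshes.
# Panning, normal-distorted smoke textures and UV sphere masks build the glow and opacity, with Fresnel and time-of-day color shaping the emissive.
# A Growth parameter blends emissive looks and shifts a curve-atlas lookup along U, which scales a Sprite function alongside a panning texture and edge mask.
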Material /BRCosmetics/Gadgets/Assets/VinderTech_GliderChute/Glider_Viper/FX/M_Viper_Mesh
graph {
 properties {
  bUsedWithMeshDeformer=true
  bUsedWithSkeletalMesh=true
  blendMode=BLEND_Translucent
  pixelDepthOffsetMode=PDOM_Legacy
  shadingModel=MSM_Unlit
  twoSided=true
  usedShadingModels=Unlit
 }
 nodes [
  n1 = SphereMask(A=ComponentMask(None=TextureCoordinate()) {g=true}, B=Constant() {r=0.5}) {attenuationRadius=0.5, hardnessPercent=0}
  n2 = TextureCoordinate() {uTiling=1.4}
  n3 = Panner(Coordinate=TextureCoordinate() {uTiling=0.8, vTiling=0.65}) {speedX=0.5, speedY=0.2}
  n4 = Multiply(A=ComponentMask(None=TextureSample(UVs=@n3) {samplerType=SAMPLERTYPE_Normal, texture=&T_Weap_Distort_N_02}.RGB) {g=true, r=true}) {constB=0.85}
  n5 = TextureSample(UVs=Panner(Coordinate=Add(A=@n2, B=@n4)) {speedX=0.4}) {samplerType=SAMPLERTYPE_Grayscale, texture=&T_Smoke_Trail_Soft}
  n6 = Multiply(A=@n1, B=Power(Base=Multiply(A=@n5.R) {constB=3.4}) {constExponent=1.25})
  n7 = LinearInterpolate(A=Constant3Vector() {constant=<rgba 0.2084 0 1 0>}, B=Constant3Vector() {constant=<rgba 0.6246 1.3372 7.5 0>}, Alpha=@n6)
  n8 = Constant() {r=-1}
  n9 = Panner(Coordinate=TextureCoordinate() {uTiling=1.4, vTiling=0.7}) {speedX=0.58}
  n10 = ComponentMask(None=TextureCoordinate()) {g=true}
  n11 = Constant() {r=0.5}
  n12 = Add(
   A=@n6
   B=Multiply(A=SphereMask(A=@n10, B=@n11) {attenuationRadius=0.21, hardnessPercent=0}, B=ScalarParameter() {defaultValue=1.5, parameterName=Param_2}) {constB=2}
  )
  n13 = Add(
   A=Multiply(
    A=Constant3Vector() {constant=<rgba 0.2369 0.9604 2 0>}
    B=TextureSample(UVs=@n9) {texture=/Game/Effects/Fort_Effects/Textures/Tile/T_Inky_Smoke_Tile2}.R
   )
   B=LinearInterpolate(A=Constant3Vector() {constant=<rgba 0.0605 0 0.1354 0>}, B=Constant3Vector() {constant=<rgba 0.0216 0.004 0.1615 0>}, Alpha=@n12)
  )
  n14 = Multiply(A=@n13) {constB=0.3}
  n15 = Power(Base=OneMinus(None=@n1)) {constExponent=2.8}
  n16 = LinearInterpolate(
   A=Constant3Vector() {constant=<rgba 0.071 0.5308 1 1>}
   B=LinearInterpolate(
    A=Constant3Vector() {constant=<rgba 0.044 0 0.0781 0>}
    B=LinearInterpolate(A=@n13, B=@n14, Alpha=@n15)
    Alpha=Fresnel() {baseReflectFraction=0, exponent=0.4}
   )
   Alpha=SphereMask(A=TextureCoordinate(), B=Constant() {r=0.5}, Radius=Constant() {r=1.25}) {attenuationRadius=0.5, hardnessPercent=20}
  )
  n17 = Constant() {r=2}
  n18 = Constant() {r=1}
  n19 = MaterialFunctionCall(Color_Morning=@n17, Color_Day=@n17, Color_Evening=@n18, Color_Night=@n18) {materialFunction=&Color_Time_of_Day}
  MaterialFunctionCall(Emissive=LinearInterpolate(A=Desaturation(None=@n7, Fraction=@n8), B=Multiply(A=@n16, B=@n19.Result), Alpha=ScalarParameter() {defaultValue=1, parameterName=Growth})) {materialFunction=&MF_FX_EmissiveAdjustments}
  n20 = SphereMask(A=ComponentMask(None=TextureCoordinate()) {r=true}, B=Constant() {r=0.5}) {attenuationRadius=0.68, hardnessPercent=0}
  n21 = Panner(Coordinate=TextureCoordinate() {uTiling=0.6, vTiling=0.45}) {speedY=0.38}
  n22 = Multiply(
   A=@n20
   B=LinearInterpolate(Alpha=TextureSample(UVs=@n21) {texture=/Game/Effects/Fort_Effects/Textures/Power_Source/T_Power_Source_Sphere_Texture}.R) {constA=0.3}
  )
  n23 = Constant() {r=32}
  n24 = TextureCoordinate()
  n25 = ScalarParameter() {defaultValue=1, parameterName=Growth}
  n26 = CurveAtlasRowParameter(CurveTime=Add(A=ComponentMask(None=@n24) {r=true}, B=LinearInterpolate(Alpha=@n25) {constA=1, constB=-1})) {
   atlas=&AthenaCurveAtlas01
   curve=/Game/Gadgets/Assets/VinderTech_GliderChute/Glider_Viper/FX/Curve_Glder_Viper
   defaultValue=134
   parameterName=Param_1
  }
  MaterialFunctionCall!(Center Location=WorldPosition().XYZ, 0-1 UVs, XY scale=Multiply(A=Multiply(A=@n22, B=@n23), B=@n26.R), OverridePivot, "Normals (Optional)") {materialFunction=&Sprite}
  Multiply(A=Multiply(A=@n12, B=ScalarParameter() {defaultValue=1, parameterName=Opacity}), B=@n20)
  ObjectPositionWS()
  VectorParameter() {defaultValue=<rgba 0 0.2085 1 1>}
  VectorParameter() {parameterName=Param_1, defaultValue=<rgba 0 0.2052 1 1>}
  VectorParameter() {parameterName=Param_2, defaultValue=<rgba 0.0114 0 0.1667 1>}
  WorldPosition()
 ]
}
